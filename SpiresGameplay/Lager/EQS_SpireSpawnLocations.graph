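# EQS query that picks candidate spawn locations for Spires, favoring points near the Athena safe zone center.
# A Fortnite LWM point provider selector generator supplies the candidate points.
# A single score-only 3D distance test against the safe zone center index context uses inverse linear scoring, so closer points rank higher and none are filtered out.
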EnvQuery /SpiresGameplay/Lager/EQS_SpireSpawnLocations
options [{
  generator {
   name=FortQueryGenerator_LWMPointProviderSelector_EQS_0
   class=FortQueryGenerator_LWMPointProviderSelector_EQS
   properties {bAutoSortTests=true, bCanRunAsync=false, envQueryResultNormalizationOption=Default, optionName=""}
  }
  tests [EnvQueryTest_Distance_0 {
    class=EnvQueryTest_Distance
    properties {
     distanceTo=/Script/FortniteAIServer.FortQueryContext_AthenaSafeZoneCenterIndex
     filterType=Range
     floatValueMin=0
     scoringEquation=InverseLinear
     scoringFactor=1
     testComment=""
     testMode=Distance3D
     testPurpose=Score
    }
   }]
 }]
